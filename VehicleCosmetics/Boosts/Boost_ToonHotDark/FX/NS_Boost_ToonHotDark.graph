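# Niagara vehicle boost cosmetic "ToonHotDark": smoke trails and glowing sparks spawned from boost sockets on the car's skeletal mesh.
# Two near-identical smoke emitters pair growing, fading meshes with short-lived R1-colored sprites, while two spark emitters add HDR-boosted sprites with curl-noise or vortex motion.
# Every emitter partitions particles across filtered sockets and applies painted-variant colors.
# Frontend-preview flags inject velocity into smoke and one spark emitter to mimic movement.
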NiagaraSystem /VehicleCosmetics/Boosts/Boost_ToonHotDark/FX/NS_Boost_ToonHotDark
dependencies {
 usedDataInterfaces [
  &NiagaraDataInterfaceCurve
  &NiagaraDataInterfaceVector2DCurve
  &NiagaraDataInterfaceArrayColor
  &NiagaraDataInterfaceSkeletalMesh
 ]
 usedDynamicInputs []
 usedModules [
  &self:SystemSpawnScript.SystemScriptSource.SystemScriptGraph.NiagaraNodeAssignment_0.SetVariables_#1
  &SystemState
 ]
}
emitters {
 NE_SmokeTrail {
  emitterUpdateScript [
   &EmitterState()
   &VC_FilteredSocketCount()
   &SpawnRate(SpawnRate=60, Use Spawn Probability=false)
   /SetVariables(
    Emitter.DiffuseColor:LinearColor=&SelectLinearColorFromArray(
     Array Sampling Mode=::NewEnumerator1
     Direct Array Index=$System.ColorVariantID
     Color Selection Array=DataInterface() {propertyValues {ColorData [
        <rgba 20 20 20 1>
        <rgba 0.1042 0.0543 0 1>
        <rgba 0 0 0 1>
        <rgba 0.0521 0.0087 0 1>
        <rgba 0 0.0048 0.06 1>
        <rgba 0.0314 0 0 1>
        <rgba 0 0.02 0.0025 1>
        <rgba 0.0781 0.0781 0.0781 1>
        <rgba 0.0333 0.05 0 0.9608>
        <rgba 0.05 0.0102 0 1>
        <rgba 0.05 0.0125 0.0453 1>
        <rgba 0.0974 0 0.1667 1>
        <rgba 0.03 0.024 0 1>
        <rgba 0 0.0471 0.05 1>
        <rgba 0.1 0.1 0.1 1>
       ]}}
     Random Seed=?
    )
    Emitter.GlowColor:LinearColor=&SelectLinearColorFromArray(
     Array Sampling Mode=::NewEnumerator1
     Direct Array Index=$System.ColorVariantID
     Color Selection Array=DataInterface() {propertyValues {ColorData [
        <rgba 20 20 20 1>
        <rgba 0.1042 0.0543 0 1>
        <rgba 0 0 0 1>
        <rgba 0.1146 0.0191 0 1>
        <rgba 10 10 10 1>
        <rgba 20 20 20 1>
        <rgba 10 10 10 1>
        <rgba 0.0781 0.0781 0.0781 1>
        <rgba 5 5 5 1>
        <rgba 5 5 5 1>
        <rgba 5 5 5 1>
        <rgba 1 1 1 1>
        <rgba 5 5 5 1>
        <rgba 5 5 5 1>
        <rgba 5 5 5 1>
       ]}}
     Random Seed=?
    )
    Emitter.Hardness:NiagaraFloat=&MakeCustomFloatFromBool(Bool=$User.bIsPainted, True Float=-1, False Float=0)
   ) {isSetParameters=true, module=SetVariables_#2}
  ]
  particleSpawnScript [
   /SetVariables(Particles.RenderVisTag:NiagaraInt32=&SpawnGroup()) {isSetParameters=true, module=SetVariables_#3}
   &VC_PaintedVariantsOverride(
    Cobalt=<rgba 0 0.0185 1 1>
    DefaultColor=<rgba 1 0.06 0 1>
    Lime=<rgba 0.4304 0.5885 0 1>
    Orange=<rgba 1 0.1217 0 1>
    Pink=<rgba 0.5052 0.0605 0.3435 1>
    Purple=<rgba 0.2789 0 0.6795 1>
    R1 Cobalt:LinearColor=<rgba 0 0.37 20 1>
    R1 Crimson:LinearColor=<rgba 2 0 0.0542 1>
    R1 DefaultColor=<rgba 1 0.2087 0 1>
    R1 ForestGreen:LinearColor=<rgba 0 2 0.1011 1>
    R1 Lime:LinearColor=<rgba 1 2 0 1>
    R1 Orange:LinearColor=<rgba 2 0.6976 0 1>
    R1 Pink:LinearColor=<rgba 2 0.2396 1.3598 1>
    R1 Purple:LinearColor=<rgba 2.0521 0 5 1>
    R1 Saffron:LinearColor=<rgba 2 1.5333 0 1>
    R1 SkyBlue:LinearColor=<rgba 0 1.3308 2 1>
    R1 TitaniumWhite:LinearColor=<rgba 2 2 2 1>
    Use R1 Painted Overrides=true
   )
   &InitializeParticle(
    Color=$Particles.Color
    Lifetime Max=1
    Lifetime Min=0.6
    Lifetime Mode=::NewEnumerator1
    Mesh Scale Mode=::NewEnumerator3
    Mesh Uniform Scale=&RandomRangeFloat(Minimum=1.3, Maximum=1.8)
    Position Mode=::NewEnumerator0
    Sprite Size Mode=::NewEnumerator3
    Uniform Sprite Size=90
    Write Material Random=true
   )
   &PartitionParticles(Number Of Partitions=$Emitter.PortCount)
   &SkeletalMeshLocation(
    Bone / Socket Index=$Particles.PartitionParticles.Partition
    Bone / Socket Sampling Mode=::NewEnumerator5
    MaskSpawnGroup=false
    Mesh Sampling Type=::NewEnumerator6
    Sampled Position Offset=&AddVectorToPosition(Position=<-20 0 0>, Vector=Linked() {linkedVariable {name=User.DTCarOffset, type=NiagaraPosition}})
    Skeletal Mesh=$System.SkeletalMesh
   )
   &ShapeLocation(MaskSpawnGroup=false, Override Seed=false, Sphere Radius=30, UseOwnerScale=true) {enabled=false}
   &AddVelocity(Velocity=&MakeCustomVectorFromBool(
     Bool=$User.bIsFrontendPreview
     True Vector=&RandomRangeVector(Minimum=<-1700 -50 -50>, Maximum=<-1500 50 50>)
     False Vector=<1 0 0>
    ))
   /VehicleCosmetics/FX/VC_FXParents/Modules/DLR_ProductPreviews(GameModeID:NiagaraInt32=$System.GameModeID, bPreviewVehicleSpeed:NiagaraBool=$System.bPreviewVehicleSpeed)
   &InitialMeshOrientation(Use Rotation Vector=false)
  ]
  particleUpdateScript [
   &ParticleState()
   &ScaleSpriteSize(Uniform Curve Sprite Scale=DataInterface() {propertyValues {Curve {
       defaultValue=1
       keys [
        {interpMode=RCIM_Cubic, tangentMode=RCTM_User, time=0, value=1}
        {
         arriveTangent=0.7114
         arriveTangentWeight=0.0814
         interpMode=RCIM_Cubic
         leaveTangent=0.7114
         leaveTangentWeight=0.3255
         tangentMode=RCTM_User
         time=0.2
         value=1.2129
        }
        {interpMode=RCIM_Cubic, tangentMode=RCTM_User, time=1, value=1.5}
       ]
      }}})
   &ScaleMeshSize(Scale Factor=&VectorFromFloat(Value=&FloatFromCurve(FloatCurve=DataInterface() {propertyValues {Curve {
         defaultValue=1
         keys [
          {interpMode=RCIM_Cubic, tangentMode=RCTM_User, time=0, value=0.6}
          {
           arriveTangent=1.7956
           arriveTangentWeight=0.0814
           interpMode=RCIM_Cubic
           leaveTangent=1.7956
           leaveTangentWeight=0.3255
           tangentMode=RCTM_User
           time=0.5938
           value=0.9906
          }
          {interpMode=RCIM_Cubic, tangentMode=RCTM_User, time=1, value=1.2776}
         ]
        }}})))
   &Drag(Drag=3, Use Linear Drag=true, Use Rotational Drag=true)
   &InheritSourceMovement(
    Applied Position Scale=&Multiply_VectorByFloat(
     Vector=<1 1 1>
     Float=&FloatFromCurve(FloatCurve=DataInterface() {propertyValues {Curve {
         defaultValue=1
         keys [{interpMode=RCIM_Cubic, tangentMode=RCTM_User, time=0, value=0}, {interpMode=RCIM_Cubic, tangentMode=RCTM_User, time=1, value=1}]
        }}})
    )
    ApplyForce=false
    ApplyPosition=true
    ApplyPositionFalloff=false
   ) {module=InheritSourceMovement001}
   &SolveForcesAndVelocity(Clamp Velocity=false, Limit Acceleration=false)
   &DynamicMaterialParameters(
    Index 0 Param 2=&FloatFromCurve(FloatCurve=DataInterface() {propertyValues {Curve {
        defaultValue=1
        keys [{interpMode=RCIM_Cubic, tangentMode=RCTM_User, time=0, value=0}, {interpMode=RCIM_Cubic, tangentMode=RCTM_User, time=1, value=1}]
       }}})
    Index 0 Param 3=0.1
    Module.Index0Float0Write=true
    Module.Index0Float1Write=true
    Module.Index0Float2Write=true
    Module.Index0Float3Write=true
   )
   /SetVariables(
    Particles.0-MeshColor:LinearColor=/DealershipSharedLibrary/FX/Niagara/DynamicInputs/MultiplyLinearColorByVectorAndFloat(
     Color to Scale:LinearColor=$Particles.Initial.Color
     RGB Scale:Vector3f=&VectorFromFloat(Value=1)
     Alpha Scale:NiagaraFloat=&FloatFromCurve(FloatCurve=DataInterface() {propertyValues {Curve {
         defaultValue=1
         keys [
          {interpMode=RCIM_Cubic, tangentMode=RCTM_User, time=0, value=1}
          {arriveTangent=-1.9937, interpMode=RCIM_Cubic, leaveTangent=-1.9937, tangentMode=RCTM_User, time=1, value=0}
         ]
        }}})
    )
    Particles.1-SpriteColor:LinearColor=&MultiplyLinearColorByVectorAndFloat(
     Color to Scale:LinearColor=$Particles.Initial.R1 Color
     RGB Scale:Vector3f=&VectorFromFloat(Value=&FloatFromCurve(FloatCurve=DataInterface() {propertyValues {Curve {
          defaultValue=1
          keys [{interpMode=RCIM_Cubic, tangentMode=RCTM_User, time=0, value=1}, {interpMode=RCIM_Cubic, tangentMode=RCTM_User, time=1, value=0}]
         }}}))
     Alpha Scale:NiagaraFloat=&FloatFromCurve(FloatCurve=DataInterface() {propertyValues {Curve {
         defaultValue=1
         keys [
          {arriveTangent=-2.5933, interpMode=RCIM_Cubic, leaveTangent=-2.5933, tangentMode=RCTM_User, time=0.225, value=1}
          {interpMode=RCIM_Cubic, tangentMode=RCTM_User, time=0.4438, value=0}
         ]
        }}})
    )
   ) {isSetParameters=true, module=SetVariables_#4}
  ]
  rendererClasses [&NiagaraMeshRendererProperties, &NiagaraSpriteRendererProperties]
  renderers [
   {rendererClass=&NiagaraMeshRendererProperties, rendererIndex=0}
   {rendererClass=&NiagaraSpriteRendererProperties, rendererIndex=1}
  ]
 }
 NE_SmokeTrail001 {
  emitterUpdateScript [
   &EmitterState()
   &VC_FilteredSocketCount()
   &SpawnRate(SpawnRate=90, Use Spawn Probability=false)
   /SetVariables(
    Emitter.DiffuseColor:LinearColor=&SelectLinearColorFromArray(
     Array Sampling Mode=::NewEnumerator1
     Direct Array Index=$System.ColorVariantID
     Color Selection Array=DataInterface() {propertyValues {ColorData [
        <rgba 20 20 20 1>
        <rgba 0.1042 0.0543 0 1>
        <rgba 0 0 0 1>
        <rgba 0.0521 0.0087 0 1>
        <rgba 0 0.0048 0.06 1>
        <rgba 0.0314 0 0 1>
        <rgba 0 0.02 0.0025 1>
        <rgba 0.0781 0.0781 0.0781 1>
        <rgba 0.0333 0.05 0 0.9608>
        <rgba 0.05 0.0102 0 1>
        <rgba 0.05 0.0125 0.0453 1>
        <rgba 0.0974 0 0.1667 1>
        <rgba 0.03 0.024 0 1>
        <rgba 0 0.0471 0.05 1>
        <rgba 0.1 0.1 0.1 1>
       ]}}
     Random Seed=?
    )
    Emitter.GlowColor:LinearColor=&SelectLinearColorFromArray(
     Array Sampling Mode=::NewEnumerator1
     Direct Array Index=$System.ColorVariantID
     Color Selection Array=DataInterface() {propertyValues {ColorData [
        <rgba 20 20 20 1>
        <rgba 0.1042 0.0543 0 1>
        <rgba 0 0 0 1>
        <rgba 0.1146 0.0191 0 1>
        <rgba 10 10 10 1>
        <rgba 20 20 20 1>
        <rgba 10 10 10 1>
        <rgba 0.0781 0.0781 0.0781 1>
        <rgba 5 5 5 1>
        <rgba 5 5 5 1>
        <rgba 5 5 5 1>
        <rgba 1 1 1 1>
        <rgba 5 5 5 1>
        <rgba 5 5 5 1>
        <rgba 5 5 5 1>
       ]}}
     Random Seed=?
    )
    Emitter.Hardness:NiagaraFloat=&MakeCustomFloatFromBool(Bool=$User.bIsPainted, True Float=-1, False Float=0)
   ) {isSetParameters=true, module=SetVariables_#2}
  ]
  particleSpawnScript [
   /SetVariables(Particles.RenderVisTag:NiagaraInt32=&SpawnGroup()) {isSetParameters=true, module=SetVariables_#3}
   &VC_PaintedVariantsOverride(
    Cobalt=<rgba 0 0.0185 1 1>
    DefaultColor=<rgba 1 0.06 0 1>
    Lime=<rgba 0.4304 0.5885 0 1>
    Orange=<rgba 1 0.1217 0 1>
    Pink=<rgba 0.5052 0.0605 0.3435 1>
    Purple=<rgba 0.2789 0 0.6795 1>
    R1 Cobalt:LinearColor=<rgba 0 0.37 20 1>
    R1 Crimson:LinearColor=<rgba 2 0 0.0542 1>
    R1 DefaultColor=<rgba 1 0.2087 0 1>
    R1 ForestGreen:LinearColor=<rgba 0 2 0.1011 1>
    R1 Lime:LinearColor=<rgba 1 2 0 1>
    R1 Orange:LinearColor=<rgba 2 0.6976 0 1>
    R1 Pink:LinearColor=<rgba 2 0.2396 1.3598 1>
    R1 Purple:LinearColor=<rgba 2.0521 0 5 1>
    R1 Saffron:LinearColor=<rgba 2 1.5333 0 1>
    R1 SkyBlue:LinearColor=<rgba 0 1.3308 2 1>
    R1 TitaniumWhite:LinearColor=<rgba 2 2 2 1>
    Use R1 Painted Overrides=true
   )
   &InitializeParticle(
    Color=$Particles.Color
    Lifetime Max=1
    Lifetime Min=0.6
    Lifetime Mode=::NewEnumerator1
    Mesh Scale Mode=::NewEnumerator3
    Mesh Uniform Scale=&RandomRangeFloat(Minimum=1.3, Maximum=1.8)
    Position Mode=::NewEnumerator0
    Sprite Size Mode=::NewEnumerator3
    Uniform Sprite Size=90
    Write Material Random=true
   )
   &PartitionParticles(Number Of Partitions=$Emitter.PortCount)
   &SkeletalMeshLocation(
    Bone / Socket Index=$Particles.PartitionParticles.Partition
    Bone / Socket Sampling Mode=::NewEnumerator5
    MaskSpawnGroup=false
    Mesh Sampling Type=::NewEnumerator6
    Sampled Position Offset=&AddVectorToPosition(Position=<-20 0 0>, Vector=Linked() {linkedVariable {name=User.DTCarOffset, type=NiagaraPosition}})
    Skeletal Mesh=$System.SkeletalMesh
   )
   &ShapeLocation(MaskSpawnGroup=false, Override Seed=false, Sphere Radius=30, UseOwnerScale=true) {enabled=false}
   &AddVelocity(Velocity=&MakeCustomVectorFromBool(
     Bool=$User.bIsFrontendPreview
     True Vector=&RandomRangeVector(Minimum=<-1700 -50 -50>, Maximum=<-1500 50 50>)
     False Vector=<1 0 0>
    ))
   &DLR_ProductPreviews(GameModeID:NiagaraInt32=$System.GameModeID, bPreviewVehicleSpeed:NiagaraBool=$System.bPreviewVehicleSpeed)
   &InitialMeshOrientation(Use Rotation Vector=false)
  ]
  particleUpdateScript [
   &ParticleState()
   &ScaleSpriteSize(Uniform Curve Sprite Scale=DataInterface() {propertyValues {Curve {
       defaultValue=1
       keys [
        {interpMode=RCIM_Cubic, tangentMode=RCTM_User, time=0, value=1}
        {
         arriveTangent=0.7114
         arriveTangentWeight=0.0814
         interpMode=RCIM_Cubic
         leaveTangent=0.7114
         leaveTangentWeight=0.3255
         tangentMode=RCTM_User
         time=0.2
         value=1.2129
        }
        {interpMode=RCIM_Cubic, tangentMode=RCTM_User, time=1, value=1.5}
       ]
      }}})
   &ScaleMeshSize(Scale Factor=&VectorFromFloat(Value=&FloatFromCurve(FloatCurve=DataInterface() {propertyValues {Curve {
         defaultValue=1
         keys [
          {interpMode=RCIM_Cubic, tangentMode=RCTM_User, time=0, value=0.6}
          {
           arriveTangent=1.7956
           arriveTangentWeight=0.0814
           interpMode=RCIM_Cubic
           leaveTangent=1.7956
           leaveTangentWeight=0.3255
           tangentMode=RCTM_User
           time=0.5938
           value=0.9906
          }
          {interpMode=RCIM_Cubic, tangentMode=RCTM_User, time=1, value=1.2776}
         ]
        }}})))
   &Drag(Drag=3, Use Linear Drag=true, Use Rotational Drag=true)
   &InheritSourceMovement(
    Applied Position Scale=&Multiply_VectorByFloat(
     Vector=<1 1 1>
     Float=&FloatFromCurve(FloatCurve=DataInterface() {propertyValues {Curve {
         defaultValue=1
         keys [{interpMode=RCIM_Cubic, tangentMode=RCTM_User, time=0, value=0}, {interpMode=RCIM_Cubic, tangentMode=RCTM_User, time=1, value=1}]
        }}})
    )
    ApplyForce=false
    ApplyPosition=true
    ApplyPositionFalloff=false
   ) {module=InheritSourceMovement001}
   &SolveForcesAndVelocity(Clamp Velocity=false, Limit Acceleration=false)
   &DynamicMaterialParameters(
    Index 0 Param 2=&FloatFromCurve(FloatCurve=DataInterface() {propertyValues {Curve {
        defaultValue=1
        keys [{interpMode=RCIM_Cubic, tangentMode=RCTM_User, time=0, value=0}, {interpMode=RCIM_Cubic, tangentMode=RCTM_User, time=1, value=1}]
       }}})
    Index 0 Param 3=0.1
    Module.Index0Float0Write=true
    Module.Index0Float1Write=true
    Module.Index0Float2Write=true
    Module.Index0Float3Write=true
   )
   /SetVariables(
    Particles.0-MeshColor:LinearColor=&MultiplyLinearColorByVectorAndFloat(
     Color to Scale:LinearColor=$Particles.Initial.Color
     RGB Scale:Vector3f=&VectorFromFloat(Value=1)
     Alpha Scale:NiagaraFloat=&FloatFromCurve(FloatCurve=DataInterface() {propertyValues {Curve {
         defaultValue=1
         keys [
          {interpMode=RCIM_Cubic, tangentMode=RCTM_User, time=0, value=1}
          {arriveTangent=-1.9937, interpMode=RCIM_Cubic, leaveTangent=-1.9937, tangentMode=RCTM_User, time=1, value=0}
         ]
        }}})
    )
    Particles.1-SpriteColor:LinearColor=&MultiplyLinearColorByVectorAndFloat(
     Color to Scale:LinearColor=$Particles.Initial.R1 Color
     RGB Scale:Vector3f=&VectorFromFloat(Value=&FloatFromCurve(FloatCurve=DataInterface() {propertyValues {Curve {
          defaultValue=1
          keys [{interpMode=RCIM_Cubic, tangentMode=RCTM_User, time=0, value=1}, {interpMode=RCIM_Cubic, tangentMode=RCTM_User, time=1, value=0}]
         }}}))
     Alpha Scale:NiagaraFloat=&FloatFromCurve(FloatCurve=DataInterface() {propertyValues {Curve {
         defaultValue=1
         keys [
          {arriveTangent=-2.5933, interpMode=RCIM_Cubic, leaveTangent=-2.5933, tangentMode=RCTM_User, time=0.225, value=1}
          {interpMode=RCIM_Cubic, tangentMode=RCTM_User, time=0.4438, value=0}
         ]
        }}})
    )
   ) {isSetParameters=true, module=SetVariables_#4}
  ]
  rendererClasses [&NiagaraMeshRendererProperties, &NiagaraSpriteRendererProperties]
  renderers [
   {rendererClass=&NiagaraMeshRendererProperties, rendererIndex=0}
   {rendererClass=&NiagaraSpriteRendererProperties, rendererIndex=1}
  ]
 }
 sparks_ {
  emitterUpdateScript [
   &EmitterState()
   &VC_FilteredSocketCount()
   &SpawnRate(SpawnRate=30, Use Spawn Probability=false)
  ]
  particleSpawnScript [
   &VC_PaintedVariantsOverride(DefaultColor=<rgba 1 0.06 0 1>)
   &InitializeParticle(
    Color=$Particles.Color
    Lifetime=&RandomRangeFloat(Minimum=0.1, Maximum=0.3)
    Sprite Rotation Mode=::NewEnumerator2
    Sprite Size Mode=::NewEnumerator3
    Uniform Sprite Size=50
    UsePositionOffset=false
    Write Material Random=true
   )
   &PartitionParticles(Number Of Partitions=$Emitter.PortCount)
   &SkeletalMeshLocation(
    Bone / Socket Index=$Particles.PartitionParticles.Partition
    Bone / Socket Sampling Mode=::NewEnumerator5
    MaskSpawnGroup=false
    Mesh Sampling Type=::NewEnumerator6
    Sampled Position Offset=&AddVectorToPosition(Position=<-20 0 0>, Vector=Linked() {linkedVariable {name=User.DTCarOffset, type=NiagaraPosition}})
    Skeletal Mesh=$System.SkeletalMesh
   )
   &AddVelocity(Velocity=<100 0 50>, Velocity Speed Scale=-1)
   &ShapeLocation(MaskSpawnGroup=false, Override Seed=false, Sphere Radius=35, UseOwnerScale=true)
   &DLR_ProductPreviews(GameModeID:NiagaraInt32=$System.GameModeID, bPreviewVehicleSpeed:NiagaraBool=$System.bPreviewVehicleSpeed)
  ]
  particleUpdateScript [
   &ParticleState()
   &ScaleSpriteSize(Uniform Curve Sprite Scale=DataInterface() {propertyValues {Curve {
       defaultValue=1
       keys [
        <at 0 0 RCIM_Cubic>
        {
         arriveTangent=1.5327
         arriveTangentWeight=0.052
         interpMode=RCIM_Cubic
         leaveTangent=1.5327
         leaveTangentWeight=0.4194
         tangentMode=RCTM_User
         time=0.1
         value=0.5
        }
        <at 1 1.2 RCIM_Cubic>
       ]
      }}})
   &ScaleColor(
    Scale Alpha=&FloatFromCurve(FloatCurve=DataInterface() {propertyValues {Curve {keys [
         {interpMode=RCIM_Cubic, tangentMode=RCTM_User, time=0, value=0}
         {interpMode=RCIM_Cubic, tangentMode=RCTM_User, time=0.1, value=1}
         {interpMode=RCIM_Cubic, tangentMode=RCTM_User, time=0.9, value=1}
         <at 1 0 RCIM_Cubic>
        ]}}})
    Scale RGB=&VectorFromFloat(Value=50)
    ScaleA=true
    ScaleRGB=true
   )
   &CurlNoiseForce(
    Noise Strength=&MergeHelper_FloatSwitch(0=1000, 1=200, 2=?, 3=?, Index=&SpawnGroup(), Num=::NewEnumerator0)
    Pan Noise=false
   ) {module=CurlNoiseForce001}
   &InheritSourceMovement(
    Applied Position Scale=&Multiply_VectorByFloat(
     Vector=<1 1 1>
     Float=&FloatFromCurve(FloatCurve=DataInterface() {propertyValues {Curve {
         defaultValue=1
         keys [
          {arriveTangent=-1.7885, interpMode=RCIM_Cubic, leaveTangent=-1.7885, tangentMode=RCTM_User, time=0.05, value=1}
          {interpMode=RCIM_Cubic, tangentMode=RCTM_User, time=1, value=0}
         ]
        }}})
    )
    ApplyForce=false
    ApplyPosition=true
    ApplyPositionFalloff=false
   ) {enabled=false, module=InheritSourceMovement001}
   &SolveForcesAndVelocity(Clamp Velocity=false, Limit Acceleration=false)
  ]
  rendererClasses [&NiagaraSpriteRendererProperties]
  renderers [{rendererClass=&NiagaraSpriteRendererProperties}]
 }
 sparks_001 {
  emitterUpdateScript [
   &EmitterState()
   &VC_FilteredSocketCount()
   &SpawnRate(SpawnRate=90, Use Spawn Probability=false)
   &SpawnRate(SpawnGroup=1, SpawnRate=5, Use Spawn Probability=false) {module=SpawnRate002}
  ]
  particleSpawnScript [
   /SetVariables(Particles.VisTag:NiagaraInt32=&SpawnGroup()) {isSetParameters=true}
   &VC_PaintedVariantsOverride(DefaultColor=<rgba 1 0.06 0 1>)
   &InitializeParticle(
    Color=$Particles.Color
    Lifetime=&RandomRangeFloat(Minimum=0.4, Maximum=0.8)
    Sprite Rotation Mode=::NewEnumerator2
    Sprite Size=<3 8>
    Sprite Size Mode=::NewEnumerator4
    UsePositionOffset=false
    Write Material Random=true
   )
   &PartitionParticles(Number Of Partitions=$Emitter.PortCount)
   &SkeletalMeshLocation(
    Bone / Socket Index=$Particles.PartitionParticles.Partition
    Bone / Socket Sampling Mode=::NewEnumerator5
    MaskSpawnGroup=false
    Mesh Sampling Type=::NewEnumerator6
    Skeletal Mesh=$System.SkeletalMesh
   )
   &AddVelocity(
    Velocity=&MakeCustomVectorFromBool(
     Bool=$User.bIsFrontendPreview
     True Vector=&RandomRangeVector(Minimum=<500 100 100>, Maximum=<1000 -100 -100>)
     False Vector=<0 0 0>
    )
    Velocity Speed Scale=-1
   )
   &ShapeLocation(MaskSpawnGroup=false, Override Seed=false, Sphere Radius=40, UseOwnerScale=true)
   &DLR_ProductPreviews(GameModeID:NiagaraInt32=$System.GameModeID, bPreviewVehicleSpeed:NiagaraBool=$System.bPreviewVehicleSpeed)
   /SetVariables(Particles.1-SpriteSize:Vector2f=<1 1>) {isSetParameters=true}
  ]
  particleUpdateScript [
   &ParticleState()
   &ScaleSpriteSize(
    Non-Uniform Curve Sprite Scale=DataInterface() {propertyValues {
      XCurve {keys [<at 0 0 RCIM_Cubic>, {arriveTangent=1, interpMode=RCIM_Cubic, leaveTangent=1, time=0.3, value=1}, <at 1 1 RCIM_Cubic>]}
      YCurve {keys [<at 0 0 RCIM_Cubic>, {arriveTangent=1, interpMode=RCIM_Cubic, leaveTangent=1, time=0.3, value=2}, <at 1 1 RCIM_Cubic>]}
     }}
    Scale Sprite Size Mode=::NewEnumerator3
   )
   &VortexForce(
    UseOriginOffset=false
    Vortex Axis=&RandomRangeVector(Minimum=<-1 -1 -1>, Maximum=<1 1 1>)
    Vortex Force Amount=&MergeHelper_FloatSwitch(0=-200, 1=0, 2=?, 3=?, Index=&SpawnGroup(), Num=::NewEnumerator0)
   )
   &CurlNoiseForce(
    Noise Strength=&MergeHelper_FloatSwitch(0=1000, 1=200, 2=?, 3=?, Index=&SpawnGroup(), Num=::NewEnumerator0)
    Pan Noise=false
   ) {enabled=false, module=CurlNoiseForce001}
   &InheritSourceMovement(
    Applied Position Scale=&Multiply_VectorByFloat(
     Vector=<1 1 1>
     Float=&FloatFromCurve(FloatCurve=DataInterface() {propertyValues {Curve {
         defaultValue=1
         keys [
          {arriveTangent=-1.7885, interpMode=RCIM_Cubic, leaveTangent=-1.7885, tangentMode=RCTM_User, time=0.05, value=1}
          {interpMode=RCIM_Cubic, tangentMode=RCTM_User, time=1, value=0}
         ]
        }}})
    )
    ApplyForce=false
    ApplyPosition=true
    ApplyPositionFalloff=false
   ) {enabled=false, module=InheritSourceMovement001}
   &ScaleColor(
    Scale Alpha=&FloatFromCurve(FloatCurve=DataInterface() {propertyValues {Curve {keys [
         {interpMode=RCIM_Cubic, tangentMode=RCTM_User, time=0, value=0}
         {interpMode=RCIM_Cubic, tangentMode=RCTM_User, time=0.1, value=1}
         {interpMode=RCIM_Cubic, tangentMode=RCTM_User, time=0.9, value=1}
         <at 1 0 RCIM_Cubic>
        ]}}})
    Scale RGB=&Multiply_VectorByFloat(Vector=<1 1 1>, Float=30)
    ScaleA=true
    ScaleRGB=true
   )
   &DynamicMaterialParameters(
    Index 0 Param 1=50
    Module.Index0Param0Write=true
    Module.Index0Param1Write=false
    Module.Index0Param2Write=false
    Module.Index0Param3Write=false
   )
   &SolveForcesAndVelocity(Clamp Velocity=false, Limit Acceleration=false)
   /SetVariables(
    Particles.1-ColorScale:LinearColor=&MultiplyLinearColorByVectorAndFloat(
     Color to Scale:LinearColor=$Particles.Initial.Color
     RGB Scale:Vector3f=&VectorFromFloat(Value=50)
     Alpha Scale:NiagaraFloat=&FloatFromCurve(FloatCurve=DataInterface() {propertyValues {Curve {
         defaultValue=1
         keys [{interpMode=RCIM_Cubic, tangentMode=RCTM_User, time=0.5, value=1}, {interpMode=RCIM_Cubic, tangentMode=RCTM_User, time=1, value=0}]
        }}})
    )
    Particles.1-SpriteScale:Vector2f=&Multiply_Vector2D_ByFloat(
     Float=&FloatFromCurve(
      FloatCurve=DataInterface() {propertyValues {Curve {
         defaultValue=1
         keys [
          <at 0 0 RCIM_Cubic>
          {
           arriveTangent=1.5327
           arriveTangentWeight=0.052
           interpMode=RCIM_Cubic
           leaveTangent=1.5327
           leaveTangentWeight=0.4194
           tangentMode=RCTM_User
           time=0.2
           value=0.5
          }
          <at 1 1.2 RCIM_Cubic>
         ]
        }}}
      Scale Curve=50
     )
     Vector2D=Linked() {linkedVariable {name=Particles.Initial.1-SpriteSize, type=Vector2f}}
    )
   ) {isSetParameters=true}
  ]
  rendererClasses [&NiagaraSpriteRendererProperties]
  renderers [{rendererClass=&NiagaraSpriteRendererProperties}]
 }
}
summary {
 systemName=NS_Boost_ToonHotDark
 emitters [
  {emitterName=sparks_001, rendererClasses [&NiagaraSpriteRendererProperties]}
  {emitterName=sparks_, rendererClasses [&NiagaraSpriteRendererProperties]}
  {emitterName=NE_SmokeTrail, rendererClasses [&NiagaraMeshRendererProperties, &NiagaraSpriteRendererProperties]}
  {
   emitterName=NE_SmokeTrail001
   rendererClasses [&NiagaraMeshRendererProperties, &NiagaraSpriteRendererProperties]
  }
 ]
 userVariables [
  User.bIsFrontendPreview {type=NiagaraBool}
  User.bIsPainted {type=NiagaraBool}
  User.ColorVariantID {type=NiagaraInt32}
  User.DTCarOffset {type=NiagaraPosition, description="", defaultValue {}}
  User.DTCarOffset001 {type=Vector3f, description="", defaultValue {struct=Vector3f, value=<0 0 0>}}
  User.DTCarOffset002 {type=Vector3f, description="", defaultValue {struct=Vector3f, value=<0 0 0>}}
  User.DTCarOffset003 {type=Vector3f, description="", defaultValue {struct=Vector3f, value=<0 0 0>}}
  User.GameModeID {type=NiagaraInt32}
  User.GlowOverride {type=NiagaraFloat, defaultValue=4, description=""}
  User.PaintedColor {type=LinearColor, description="", defaultValue {struct=LinearColor, value=<rgba 1 1 1 1>}}
  User.PaintedScalar {type=NiagaraFloat, defaultValue=1, description=""}
  User.PortCount {type=NiagaraInt32, defaultValue=1}
  User.SingleBoostPos {type=NiagaraPosition, description="", defaultValue {}}
 ]
}
